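# Animation Blueprint for the Keen backpack cosmetic, adding secondary motion to the pack and its two arrow bones.
# Three chained Anim Dynamics nodes at partial alpha sway pack, dyn_arrow_1 and dyn_arrow_2 from the reference pose within tight angular limits.
# An additive bone-space pitch then tilts the pack, weighted by SkydiveFloat.
# The update event switches that weight between zero and one from bIsSkydiveFloating, without interpolation.
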AnimBlueprint /BRCosmetics/Accessories/FORT_Backpacks/Backpack_M_MED_Keen/Meshes/M_MED_Keen_Pack_AnimBP
target_skeleton=&/M_MED_Keen_Pack_Skeleton
graphs [
 AnimGraph {nodes [
   n1 = Anim Dynamics - Bone: pack(ComponentPose=Local To Component(LocalPose=Local Space Ref Pose()), Alpha=0.3, ExternalForce="0,0,0") {
    angularDampingOverride=1
    bAngularSpring=true
    bOverrideAngularDamping=true
    linearSpringConstant=0.1
    alphaBoolBlend {blendOption=Linear}
    boundBone {boneName=pack}
    physicsBodyDefinitions [{
      sphereCollisionRadius=0
      boundBone {boneName=pack}
      boxExtents=<20 20 20>
      constraintSetup {coneAngle=5, angularLimitsMax=<2 8 2>, angularLimitsMin=<-2 -8 -2>, angularTarget=<1 0 0>}
      localJointOffset=<0 0 -20>
     }]
    relativeSpaceBone {boneName=root}
   }
   n2 = Anim Dynamics - Bone: dyn_arrow_1(ComponentPose=@n1, Alpha=0.3, ExternalForce="0,0,0") {
    angularDampingOverride=1
    bAngularSpring=true
    bLinearSpring=true
    bOverrideAngularDamping=true
    linearSpringConstant=0.1
    alphaBoolBlend {blendOption=Linear}
    boundBone {boneName=dyn_arrow_1}
    physicsBodyDefinitions [{
      sphereCollisionRadius=0
      boundBone {boneName=dyn_arrow_1}
      boxExtents=<15 15 15>
      constraintSetup {coneAngle=5, angularLimitsMax=<2 5 2>, angularLimitsMin=<-2 -5 -2>, angularTarget=<1 0 0>, linearAxesMax=<10 0 0>}
      localJointOffset=<-15 0 -5>
     }]
   }
   n3 = Anim Dynamics - Bone: dyn_arrow_2(ComponentPose=@n2, Alpha=0.3, ExternalForce="0,0,0") {
    angularDampingOverride=1
    bOverrideAngularDamping=true
    linearSpringConstant=0.1
    alphaBoolBlend {blendOption=Linear}
    boundBone {boneName=dyn_arrow_2}
    physicsBodyDefinitions [{
      sphereCollisionRadius=0
      boundBone {boneName=dyn_arrow_2}
      boxExtents=<20 20 20>
      constraintSetup {coneAngle=5, angularLimitsMax=<2 5 2>, angularLimitsMin=<-2 -5 -2>, angularTarget=<1 0 0>, linearAxesMax=<10 0 0>}
      localJointOffset=<-13 0 -5>
     }]
   }
   n4 = "Transform (Modify) Bone - Bone: pack"(ComponentPose=@n3, Alpha=Get SkydiveFloat() -> (SkydiveFloat:double), Translation="0,0,0", Rotation="-20,0,0") {
    rotationMode=BMM_Additive
    rotationSpace=BCS_BoneSpace
    alphaBoolBlend {blendOption=Linear}
    boneToModify {boneName=pack}
    rotation {pitch=-20}
   }
   Output Pose(Result=Component To Local(ComponentPose=@n4)) {layerGroup=None}
  ]}
 EventGraph {nodes [
   Event BlueprintUpdateAnimation()
   > Set SkydiveFloat(SkydiveFloat:double=SelectFloat(A=1, B=0, bPickA=Get bIsSkydiveFloating())) -> (Output_Get:double)
   TryGetPawnOwner()
  ]}
]
variables [SkydiveFloat {type=number}]
